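# Control Rig FK module "Meta" that builds rotator controls for bones on four connectors, all under the first bone's parent.
# Construction adds a shared Spread channel and scales controls by chain length.
# Forward solve drives bones from controls; with Spread on, interactively moving either end control spreads a ratio-weighted share of its change since the cached previous transform across the others, and control shapes switch.
# Backward solve snaps controls to bones.
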ControlRigBlueprint /EpicControlRig/Modules/FkSolves/Meta/CRM_Epic_Meta_v01
graphs [
 ConstructionGraph {nodes [
   RigUnit_PrepareForExecution = &RigUnit_PrepareForExecution()
   RigUnit_ItemArray = &RigUnit_ItemArray(
    Items=((Type=Connector,Name="Bone A"),(Type=Connector,Name="Bone B"),(Type=Connector,Name="Bone C"),(Type=Connector,Name="Bone D"))
    Items.0=/Script/ControlRig.RigUnit_ResolveConnector(Connector=(Type=Connector,Name="Bone A"), SkipSocket=False) {is=struct RigVMUnitNode}.Result
    Items.1=&RigUnit_ResolveConnector(Connector=(Type=Connector,Name="Bone B"), SkipSocket=False).Result
    Items.2=&RigUnit_ResolveConnector(Connector=(Type=Connector,Name="Bone C"), SkipSocket=False).Result
    Items.3=&RigUnit_ResolveConnector(Connector=(Type=Connector,Name="Bone D"), SkipSocket=False).Result
   )
   RigVMVariableNode(Value=@RigUnit_ItemArray.Items, ExecuteContext=@RigUnit_PrepareForExecution.ExecutePin) {variable=Bones}
   DISPATCH_RigVMDispatch_ArrayGetAtIndex = RigVMDispatchNode(Index=0, Array=@RigUnit_ItemArray.Items) {template="DISPATCH_RigVMDispatch_ArrayGetAtIndex(in Array,in Index,out Element)"}
   > RigVMVariableNode(Value=/Script/ControlRig.RigUnit_HierarchyGetParent(bDefaultParent=True, Child=@DISPATCH_RigVMDispatch_ArrayGetAtIndex.Element) {is=struct RigVMUnitNode}.Parent) {variable=Parent}
   > DISPATCH_RigVMDispatch_ArrayReset = RigVMDispatchNode(Array=RigVMVariableNode() {variable=Controls}.Value) {template="DISPATCH_RigVMDispatch_ArrayReset(io Array)"}
   > RigVMFunction_Sequence = RigVMAggregateNode()
   For_Each_1 = RigVMDispatchNode(Array=RigVMVariableNode() {variable=Bones}.Value, ExecuteContext=@RigVMFunction_Sequence.A) {template="DISPATCH_RigVMDispatch_ArrayIterator(in Array,out Element,out Index,out Count,out Ratio)"}
   > Get Item Name v01 = RigVMFunctionReferenceNode(
    Numerical Name Start Index=1
    Prefix=FK
    Remove Side Suffix=true
    Standard Numerical Names=true
    Count=@For_Each_1.Count
    Name=@For_Each_1.Element.Name
    Index=@For_Each_1.Index
   )
   n1 = RigVMVariableNode() {variable=Shape Settings}
   Get Control Color From Metadata = RigVMFunctionReferenceNode(Color Override=@n1.Value.Color)
   Make_Array = RigVMDispatchNode(
    Values=("None","None","None","None")
    Values.0=RigVMVariableNode() {variable=Meta A Display Name}.Value
    Values.1=RigVMVariableNode() {variable=Meta B Display Name}.Value
    Values.2=RigVMVariableNode() {variable=Meta C Display Name}.Value
    Values.3=RigVMVariableNode() {variable=Meta D Display Name}.Value
   ) {template="DISPATCH_RigVMDispatch_ArrayMake(in Values,out Array)"}
   > SpawnControl = /Script/ControlRig.RigUnit_HierarchyAddControlRotator(
    InitialValue=(Pitch=0,Yaw=0,Roll=0)
    OffsetSpace=GlobalSpace
    Settings="(InitialSpace=LocalSpace,Limits=(LimitPitch=(bMinimum=False,bMaximum=False),LimitYaw=(bMinimum=False,bMaximum=False),LimitRoll=(bMinimum=False,bMaximum=False),MinValue=(Pitch=-180,Yaw=-180,Roll=-180),MaxValue=(Pitch=180,Yaw=180,Roll=180),bDrawLimits=true), ...(truncated)"
    Settings.Shape.Color=@"Get Control Color From Metadata".Result
    Name=@"Get Item Name v01".Result
    OffsetTransform=/Script/RigVM.RigVMFunction_MathTransformMakeAbsolute(
     Parent=&RigUnit_GetTransform(bInitial=true, Item=@For_Each_1.Element).Transform
     Local=RigVMVariableNode() {variable=Control Transform Offset}.Value
    ) {is=struct RigVMUnitNode, template="Make Absolute::Execute(in Local,in Parent,out Global)"}.Global
    Parent=RigVMVariableNode() {variable=Parent}.Value
    Settings.Shape.Name=@n1.Value.Name
    Settings.Shape.Transform=@n1.Value.Transform
    Settings.Shape.bVisible=@n1.Value.bVisible
    Settings.DisplayName=RigVMDispatchNode(Index=@For_Each_1.Index, Array=@Make_Array.Array) {template="DISPATCH_RigVMDispatch_ArrayGetAtIndex(in Array,in Index,out Element)"}.Element
   ) {
    is=struct RigVMUnitNode
    template="SpawnControl::Execute(in InitialValue,in Settings,in OffsetTransform,in OffsetSpace,in Parent,in Name,out Item)"
   }
   > RigVMDispatchNode(Array=RigVMVariableNode() {variable=Controls}.Value, Element=@SpawnControl.Item) {template="DISPATCH_RigVMDispatch_ArrayAdd(io Array,in Element,out Index)"}
   > RigVMDispatchNode(
    Name=Previous Value
    NameSpace=Self
    Value=(Rotation=(X=0,Y=0,Z=0,W=1),Translation=(X=0,Y=0,Z=0),Scale3D=(X=1,Y=1,Z=1))
    Item=@SpawnControl.Item
   ) {template="DISPATCH_RigDispatch_SetMetadata(in Item,in Name,in NameSpace,in Value,out Success)"}
   > Set Mirror Behavior = RigVMFunctionReferenceNode(Value=RigVMVariableNode() {variable=Mirror Behavior}.Value, Item=@SpawnControl.Item)
   > Set Mirror Axis = RigVMFunctionReferenceNode(Value=RigVMVariableNode() {variable=Mirror Axis}.Value, Item=@SpawnControl.Item)
   n2 = RigVMVariableNode() {variable=Bones}
   Get Chain Length = RigVMFunctionReferenceNode(Initial=true, Items=@n2.Value, ExecuteContext=@RigVMFunction_Sequence.D)
   Get_Module_Metadata_1 = RigVMDispatchNode(Default=1, Name=Global Control Scale, NameSpace=Root) {template="DISPATCH_RigDispatch_GetModuleMetadata(in Name,in NameSpace,in Default,out Value,out Found)"}
   Num = RigVMDispatchNode(Array=@n2.Value) {template="DISPATCH_RigVMDispatch_ArrayGetNum(in Array,out Num)"}
   > Set Control Scale = RigVMFunctionReferenceNode(
    Global Scale=@Get_Module_Metadata_1.Value
    Auto Scale Factor=RigVMDispatchNode(
     False=1
     True=&RigVMFunction_MathDoubleDiv(B=4.8, A=@"Get Chain Length".Length).Result
     Condition=/Script/RigVM.RigVMFunction_MathIntGreater(B=1, A=@Num.Num) {is=struct RigVMUnitNode, template="Greater::Execute(in A,in B,out Result)"}.Result
    ) {template="DISPATCH_RigVMDispatch_If(in Condition,in True,in False,out Result)"}.Result
    Controls=RigVMVariableNode() {variable=Controls}.Value
    Scale=RigVMVariableNode() {variable=Control Scale}.Value
   )
   SpawnAnimationChannel = /Script/ControlRig.RigUnit_HierarchyAddAnimationChannelBool(
    LimitsEnabled=()
    MaximumValue=True
    MinimumValue=False
    Name=Spread
    Parent=RigVMDispatchNode(Index=-1, Array=RigVMVariableNode() {variable=Controls}.Value) {template="DISPATCH_RigVMDispatch_ArrayGetAtIndex(in Array,in Index,out Element)"}.Element
    InitialValue=RigVMVariableNode() {variable=Default Spread}.Value
    ExecutePin=@RigVMFunction_Sequence.B
   ) {
    is=struct RigVMUnitNode
    template="SpawnAnimationChannel::Execute(in InitialValue,in MinimumValue,in MaximumValue,in LimitsEnabled,in Parent,in Name,out Item)"
   }
   Clone = RigVMDispatchNode(Array=RigVMVariableNode() {variable=Controls}.Value) {template="DISPATCH_RigVMDispatch_ArrayClone(in Array,out Clone)"}
   > Remove = RigVMDispatchNode(Index=-1, Array=@Clone.Clone) {template="DISPATCH_RigVMDispatch_ArrayRemove(io Array,in Index)"}
   > /Script/ControlRig.RigUnit_SetChannelHosts(Hosts=@Remove.Array, Channel=@SpawnAnimationChannel.Item) {is=struct RigVMUnitNode}
   n3 = RigVMVariableNode() {variable=Controls}
   SetControlDrivenList = /Script/ControlRig.RigUnit_SetControlDrivenList(
    Control=/Script/ControlRig.RigUnit_ItemToName(Value=RigVMDispatchNode(Index=0, Array=@n3.Value) {template="DISPATCH_RigVMDispatch_ArrayGetAtIndex(in Array,in Index,out Element)"}.Element) {is=struct RigVMUnitNode, template="Cast::Execute(in Value,out Result)"}.Result
    Driven=@n3.Value
    ExecutePin=@RigVMFunction_Sequence.C
   ) {is=struct RigVMUnitNode, template="SetControlDrivenList::Execute(in Control,in Driven)"}
   RigUnit_ItemToName = &RigUnit_ItemToName(Value=RigVMDispatchNode(Index=-1, Array=@n3.Value) {template="DISPATCH_RigVMDispatch_ArrayGetAtIndex(in Array,in Index,out Element)"}.Element) {template="Cast::Execute(in Value,out Result)"}
   > &RigUnit_SetControlDrivenList(Control=@RigUnit_ItemToName.Result, Driven=@n3.Value) {template="SetControlDrivenList::Execute(in Control,in Driven)"}
  ]}
 New Function_ContainedGraph {nodes [
   n1 = RigVMFunctionEntryNode()
   RigVMFunction_MathBoolToInteger = &RigVMFunction_MathBoolToInteger(Value=@n1.Value)
   DISPATCH_RigVMDispatch_SelectInt32 = RigVMDispatchNode(Values=("Sphere_Solid","Box_Solid"), Values.0=@n1.False Name, Values.1=@n1.True Name, Index=@RigVMFunction_MathBoolToInteger.Result) {template="DISPATCH_RigVMDispatch_SelectInt32(in Index,in Values,out Result)"}
   n2 = /Script/ControlRig.RigUnit_HierarchyGetShapeSettings(Item=@n1.Item) {is=struct RigVMUnitNode}
   > /Script/ControlRig.RigUnit_HierarchySetShapeSettings(
    Settings=(bVisible=True,Name="Default",Color=(R=1,G=0,B=0,A=1),Transform=(Rotation=(X=0,Y=0,Z=0,W=1),Translation=(X=0,Y=0,Z=0),Scale3D=(X=1,Y=1,Z=1)))
    Settings.Name=@DISPATCH_RigVMDispatch_SelectInt32.Result
    Item=@n1.Item
    Settings.Color=@n2.Settings.Color
    Settings.Transform=@n2.Settings.Transform
    Settings.bVisible=@n2.Settings.bVisible
   ) {is=struct RigVMUnitNode}
   > RigVMFunctionReturnNode()
  ]}
 RigVMFunctionLibrary {nodes [Select Control Shape = RigVMCollapseNode(ExecuteContext=(), False Name=Sphere_Solid, Item=(Type=Control,Name="None"), True Name=Box_Solid, Value=False)]}
 RigVMFunction_Sequence_1_ContainedGraph {nodes [
   n1 = RigVMFunctionEntryNode()
   RigVMFunction_Sequence = &RigVMFunction_Sequence(ExecuteContext=@n1.ExecuteContext)
   n2 = &RigVMFunction_Sequence(ExecuteContext=@RigVMFunction_Sequence.B)
   n3 = &RigVMFunction_Sequence(ExecuteContext=@n2.B)
   RigVMFunctionReturnNode(A=@RigVMFunction_Sequence.A, B=@n2.A, C=@n3.A, D=@n3.B)
  ]}
 RigVMFunction_Sequence_1_ContainedGraph {nodes [
   n1 = RigVMFunctionEntryNode()
   RigVMFunction_Sequence = &RigVMFunction_Sequence(ExecuteContext=@n1.ExecuteContext)
   n2 = &RigVMFunction_Sequence(ExecuteContext=@RigVMFunction_Sequence.B)
   n3 = &RigVMFunction_Sequence(ExecuteContext=@n2.B)
   RigVMFunctionReturnNode(A=@RigVMFunction_Sequence.A, B=@n2.A, C=@n3.A, D=@n3.B)
  ]}
 RigVMModel {nodes [
   RigUnit_BeginExecution = &RigUnit_BeginExecution()
   n1 = RigVMVariableNode() {variable=Controls}
   RigVMVariableNode(
    ExecuteContext=@RigUnit_BeginExecution.ExecutePin
    Value=/Script/ControlRig.RigUnit_GetBoolAnimationChannel(
     Channel=Spread
     bInitial=False
     Control=RigVMDispatchNode(Index=-1, Array=@n1.Value) {template="DISPATCH_RigVMDispatch_ArrayGetAtIndex(in Array,in Index,out Element)"}.Element.Name
    ) {is=struct RigVMUnitNode, template="GetAnimationChannel::Execute(out Value,in Control,in Channel,in bInitial)"}.Value
   ) {variable=Auto Spread}
   > RigVMFunction_Sequence = RigVMAggregateNode()
   DISPATCH_RigVMDispatch_ArrayIterator = RigVMDispatchNode(Array=RigVMVariableNode() {variable=Controls}.Value, ExecuteContext=@RigVMFunction_Sequence.C) {template="DISPATCH_RigVMDispatch_ArrayIterator(in Array,out Element,out Index,out Count,out Ratio)"}
   > DISPATCH_RigDispatch_SetMetadata = RigVMDispatchNode(
    Name=Previous Value
    NameSpace=Self
    Item=@DISPATCH_RigVMDispatch_ArrayIterator.Element
    Value=&RigUnit_GetTransform(Space=LocalSpace, Item=@DISPATCH_RigVMDispatch_ArrayIterator.Element).Transform
   ) {template="DISPATCH_RigDispatch_SetMetadata(in Item,in Name,in NameSpace,in Value,out Success)"}
   For_Each = RigVMDispatchNode(Array=RigVMVariableNode() {variable=Bones}.Value, ExecuteContext=@RigVMFunction_Sequence.B) {template="DISPATCH_RigVMDispatch_ArrayIterator(in Array,out Element,out Index,out Count,out Ratio)"}
   n2 = RigVMDispatchNode(Index=@For_Each.Index, Array=RigVMVariableNode() {variable=Controls}.Value) {template="DISPATCH_RigVMDispatch_ArrayGetAtIndex(in Array,in Index,out Element)"}
   Inverse = /Script/RigVM.RigVMFunction_MathTransformInverse(Value=RigVMVariableNode() {variable=Control Transform Offset}.Value) {is=struct RigVMUnitNode, template="Inverse::Execute(in Value,out Result)"}
   > &RigUnit_SetTransform(
    Weight=1
    bInitial=False
    Item=@For_Each.Element
    Value=&RigVMFunction_MathTransformMakeAbsolute(Parent=&RigUnit_GetTransform(bInitial=False, Item=@n2.Element).Transform, Local=@Inverse.Result) {template="Make Absolute::Execute(in Local,in Parent,out Global)"}.Global
   )
   n3 = RigVMDispatchNode(Array=RigVMVariableNode() {variable=Controls}.Value, ExecuteContext=@RigVMFunction_Sequence.D) {template="DISPATCH_RigVMDispatch_ArrayIterator(in Array,out Element,out Index,out Count,out Ratio)"}
   > Select Control Shape_1 = RigVMFunctionReferenceNode(
    True Name=Default
    Item=@n3.Element
    False Name=RigVMVariableNode() {variable=Shape Settings}.Value.Name
    Value=RigVMVariableNode() {variable=Auto Spread}.Value
   )
   n4 = /Script/ControlRig.RigUnit_IsInteracting() {is=struct RigVMUnitNode}
   RigVMFunction_MathBoolAnd = &RigVMFunction_MathBoolAnd(A=@n4.bIsInteracting, B=RigVMVariableNode() {variable=Auto Spread}.Value)
   RigVMFunction_ControlFlowBranch = &RigVMFunction_ControlFlowBranch(Condition=@RigVMFunction_MathBoolAnd.Result, ExecuteContext=@RigVMFunction_Sequence.A)
   DISPATCH_RigVMDispatch_ArrayGetAtIndex = RigVMDispatchNode(Index=0, Array=@n4.Items) {template="DISPATCH_RigVMDispatch_ArrayGetAtIndex(in Array,in Index,out Element)"}
   n5 = RigVMVariableNode() {variable=Controls}
   DISPATCH_RigVMDispatch_CoreEquals = RigVMDispatchNode(
    A=@DISPATCH_RigVMDispatch_ArrayGetAtIndex.Element
    B=RigVMDispatchNode(Index=0, Array=@n5.Value) {template="DISPATCH_RigVMDispatch_ArrayGetAtIndex(in Array,in Index,out Element)"}.Element
   ) {template="DISPATCH_RigVMDispatch_CoreEquals(in A,in B,out Result)"}
   RigVMFunction_MathBoolOr = &RigVMFunction_MathBoolOr(
    A=@DISPATCH_RigVMDispatch_CoreEquals.Result
    B=RigVMDispatchNode(
     B=@DISPATCH_RigVMDispatch_ArrayGetAtIndex.Element
     A=RigVMDispatchNode(Index=-1, Array=@n5.Value) {template="DISPATCH_RigVMDispatch_ArrayGetAtIndex(in Array,in Index,out Element)"}.Element
    ) {template="DISPATCH_RigVMDispatch_CoreEquals(in A,in B,out Result)"}.Result
   )
   n6 = &RigVMFunction_ControlFlowBranch(ExecuteContext=@RigVMFunction_ControlFlowBranch.True, Condition=@RigVMFunction_MathBoolOr.Result)
   RigVMVariableNode(Value=@DISPATCH_RigVMDispatch_CoreEquals.Result, ExecuteContext=@n6.True) {variable=Is First Control}
   > n7 = RigVMDispatchNode(Array=RigVMVariableNode() {variable=Controls}.Value) {template="DISPATCH_RigVMDispatch_ArrayIterator(in Array,out Element,out Index,out Count,out Ratio)"}
   Subtract = &RigVMFunction_MathDoubleSub(A=1, B=@n7.Ratio)
   DISPATCH_RigVMDispatch_If = RigVMDispatchNode(False=@n7.Ratio, True=@Subtract.Result, Condition=RigVMVariableNode() {variable=Is First Control}.Value) {template="DISPATCH_RigVMDispatch_If(in Condition,in True,in False,out Result)"}
   n8 = &RigVMFunction_ControlFlowBranch(
    ExecuteContext=@n7.ExecuteContext
    Condition=RigVMDispatchNode(B=1, A=@DISPATCH_RigVMDispatch_If.Result) {template="DISPATCH_RigVMDispatch_CoreEquals(in A,in B,out Result)"}.Result
   )
   DISPATCH_RigDispatch_GetMetadata = RigVMDispatchNode(
    Default=(Rotation=(X=0,Y=0,Z=0,W=1),Translation=(X=0,Y=0,Z=0),Scale3D=(X=1,Y=1,Z=1))
    Name=Previous Value
    NameSpace=Self
    Item=@DISPATCH_RigVMDispatch_ArrayGetAtIndex.Element
   ) {template="DISPATCH_RigDispatch_GetMetadata(in Item,in Name,in NameSpace,in Default,out Value,out Found)"}
   Interpolate = /Script/RigVM.RigVMFunction_MathTransformLerp(
    A=(Rotation=(X=0,Y=0,Z=0,W=1),Translation=(X=0,Y=0,Z=0),Scale3D=(X=1,Y=1,Z=1))
    T=@DISPATCH_RigVMDispatch_If.Result
    B=&RigVMFunction_MathTransformMakeRelative(
     Parent=@DISPATCH_RigDispatch_GetMetadata.Value
     Global=&RigUnit_GetTransform(Space=LocalSpace, Item=@DISPATCH_RigVMDispatch_ArrayGetAtIndex.Element).Transform
    ).Local
   ) {is=struct RigVMUnitNode, template="Interpolate::Execute(in A,in B,in T,out Result)"}
   &RigUnit_OffsetTransformForItem(Weight=1, bPropagateToChildren=True, Item=@n7.Element, OffsetTransform=@Interpolate.Result, ExecutePin=@n8.False)
  ]}
 RigVMModel Backwards Solve Graph {nodes [
   n1 = /Script/ControlRig.RigUnit_InverseExecution() {is=struct RigVMUnitNode}
   For_Each = RigVMDispatchNode(ExecuteContext=@n1.ExecutePin, Array=RigVMVariableNode() {variable=Bones}.Value) {template="DISPATCH_RigVMDispatch_ArrayIterator(in Array,out Element,out Index,out Count,out Ratio)"}
   > &RigUnit_SetTransform(
    Weight=1
    bInitial=False
    Value=&RigVMFunction_MathTransformMakeAbsolute(
     Parent=&RigUnit_GetTransform(bInitial=False, Item=@For_Each.Element).Transform
     Local=RigVMVariableNode() {variable=Control Transform Offset}.Value
    ) {template="Make Absolute::Execute(in Local,in Parent,out Global)"}.Global
    Item=RigVMDispatchNode(Index=@For_Each.Index, Array=RigVMVariableNode() {variable=Controls}.Value) {template="DISPATCH_RigVMDispatch_ArrayGetAtIndex(in Array,in Index,out Element)"}.Element
   )
  ]}
 RigVMModel Connector Event Graph {nodes [/Script/ControlRig.RigUnit_ConnectorExecution() {is=struct RigVMUnitNode}]}
]
hierarchy [
 Bone A {type=Connector, loc=<0 0 0>, rot=<0 0 0 1>}
 Root {type=Socket, loc=<0 0 0>, rot=<0 0 0 1>}
 Bone C {type=Connector, loc=<0 0 0>, rot=<0 0 0 1>}
 Bone B {type=Connector, loc=<0 0 0>, rot=<0 0 0 1>}
 Bone D {type=Connector, loc=<0 0 0>, rot=<0 0 0 1>}
]
variables [
 Auto Spread {type=bool, default=False}
 Bones {type="TArray<FRigElementKey>"}
 Control Scale {type=double, default=1}
 Control Transform Offset {type=FTransform, default=(Rotation=(X=0,Y=0,Z=0,W=1),Translation=(X=0,Y=0,Z=0),Scale3D=(X=1,Y=1,Z=1))}
 Controls {type="TArray<FRigElementKey>"}
 Default Spread {type=bool, default=False}
 Is First Control {type=bool, default=False}
 Meta A Display Name {type=FName, default=None}
 Meta B Display Name {type=FName, default=None}
 Meta C Display Name {type=FName, default=None}
 Meta D Display Name {type=FName, default=None}
 Mirror Axis {type=FVector, default=(X=0,Y=1,Z=1)}
 Mirror Behavior {type=bool, default=False}
 Parent {type=FRigElementKey, default=(Type=None,Name="")}
 Shape Settings {
  type=FRigUnit_HierarchyAddControl_ShapeSettings
  default=(bVisible=True,Name="Default",Color=(R=1,G=1,B=1,A=1),Transform=(Rotation=(X=0,Y=0,Z=0,W=1),Translation=(X=0,Y=2,Z=0),Scale3D=(X=0.08,Y=0.08,Z=0.08)))
 }
]
